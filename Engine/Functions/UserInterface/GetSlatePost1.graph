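# Library material function for UI materials that reads Slate's post-processed backbuffer copy 1, which exists only while a widget using the material is present.
# It samples the SlatePost1 texture parameter, defaulting to the SlatePost1_RT render target, at optional UVs that fall back to screen-position viewport UVs.
# It outputs raw RGB and RGBA, plus linear versions built by running each channel, alpha included, through sRGBToLinear.
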MaterialFunction /Engine/Functions/UserInterface/GetSlatePost1
graph {
 properties {
  bExposeToLibrary=true
  description=Reference to slate postprocessed copy 1 of backbuffer. Copy only occurs if widget using material is present. Space is freed when not in use.
  libraryCategoriesText [User Interface, Misc]
 }
 nodes [
  n1 = FunctionInput(Preview=ScreenPosition().ViewportUV) {
   bUsePreviewValueAsDefault=true
   description="UV to read slate post buffer with, suggest using 'ScreenPosition' as base for any manipulation. Default is 'ScreenPosition'"
   inputName="UVs (Optional)"
   inputType=FunctionInput_Vector2
  }
  n2 = TextureSampleParameter2D(UVs=@n1) {parameterName=SlatePost1, samplerType=SAMPLERTYPE_LinearColor, texture=/Engine/EngineResources/SlatePost1_RT}
  FunctionOutput(None=@n2.RGB) {outputName=RGB}
  FunctionOutput(None=@n2.RGBA) {outputName=RGBA, sortPriority=2}
  n3 = MaterialFunctionCall!(sRGB=@n2.R) {materialFunction=/Engine/Functions/Engine_MaterialFunctions01/ImageAdjustment/sRGBToLinear}
  n4 = MaterialFunctionCall(sRGB=@n2.G) {materialFunction=&sRGBToLinear}
  n5 = MaterialFunctionCall(sRGB=@n2.B) {materialFunction=&sRGBToLinear}
  FunctionOutput(None=MaterialFunctionCall(X=@n3.Linear, Y=@n4.Linear, Z=@n5.Linear) {materialFunction=&MakeFloat3}.Result) {outputName=LinearRGB, sortPriority=1}
  FunctionOutput(None=MaterialFunctionCall(
    X=@n3.Linear
    Y=@n4.Linear
    Z=@n5.Linear
    A=MaterialFunctionCall(sRGB=@n2.A) {materialFunction=&sRGBToLinear}.Linear
   ) {materialFunction=&MakeFloat4}.Result) {outputName=LinearRGBA, sortPriority=3}
 ]
}
