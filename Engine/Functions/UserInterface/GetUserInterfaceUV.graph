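# Material function that exposes the built-in UV sets of Slate and UMG elements to UI materials.
# Each output passes a texture coordinate channel straight through, with no math.
# Channels 0 and 1 give 9-slice UVs with and without tiling, falling back to 0-1 UVs without a 9-slice.
# Channels 2, 3 and 4 give tiling, screen-space pixel size and normalized UVs spanning the whole element.
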MaterialFunction /Engine/Functions/UserInterface/GetUserInterfaceUV
graph {
 properties {
  bExposeToLibrary=true
  description=Provides access to various built in UV sets for UI materials rendered with Slate or UMG
  libraryCategoriesText [UserInterface, Misc]
 }
 nodes [
  FunctionOutput(None=TextureCoordinate()) {
   description="The computed UVs for a 9-slice box with tiling information.  If the element has no 9 slice then this just returns a tiled 0-1 uv set"
   outputName=9-Slice UV
  }
  FunctionOutput(None=TextureCoordinate() {coordinateIndex=2}) {description=The computed horizontal and vertical tiling for this element, outputName=Tiling, sortPriority=3}
  FunctionOutput(None=TextureCoordinate() {coordinateIndex=1}) {
   description="The computed UVs for a 9-slice box without tiling information.  If the element has no 9 slice then this just returns a 0-1 uv set"
   outputName="9-Slice UV (No Tiling)"
   sortPriority=4
  }
  FunctionOutput(None=TextureCoordinate() {coordinateIndex=3}) {description=The screen space size of the element being drawn, outputName=Pixel Size, sortPriority=1}
  FunctionOutput(None=TextureCoordinate() {coordinateIndex=4}) {description=A 0-1 UV set that spans the entire element, outputName=Normalized UV, sortPriority=2}
 ]
}
